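# Niagara system for the Boost_EP vehicle boost cosmetic, a green-tinted exhaust effect at filtered skeletal-mesh sockets, shifted by a per-car offset.
# A multi-renderer ribbon trail combines with short mesh-and-sprite bursts, flares and a backward cone of splash sprites.
# Another sprite emitter uses spawn groups for large growing and small shrinking particles that accelerate upward.
# Particles are partitioned across ports; painted-variant overrides set colours on all emitters except that rising one.
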NiagaraSystem /VehicleCosmetics/Boosts/Boost_EP/FX/NS_Boost_EP
dependencies {
 usedDataInterfaces [
  &NiagaraDataInterfaceCurve
  &NiagaraDataInterfaceSpriteRendererInfo
  &NiagaraDataInterfaceSkeletalMesh
 ]
 usedDynamicInputs []
 usedModules [
  &self:SystemSpawnScript.SystemScriptSource.SystemScriptGraph.NiagaraNodeAssignment_0.SetVariables_#1
  &SystemState
 ]
}
emitters {
 NE_VC_Boost {
  emitterUpdateScript [
   &EmitterState(Inactive Response=::NewEnumerator1, Life Cycle Mode=::NewEnumerator1, Loop Duration=1, UseLoopDelay=false)
   &VC_FilteredSocketCount()
   &VC_RateQualityHandler(Epic Spawn Rate=40, High Spawn Rate=40, Low Spawn Rate=30, Use Custom Rate?=true)
   &VC_SpawnRibbon_v1(Number of Ribbons=$Emitter.PortCount, SpawnRate=$Emitter.Spawn Rate)
  ]
  particleSpawnScript [
   &VC_PaintedVariantsOverride(
    DefaultColor=<rgba 0.0307 0.16 0.0048 0.94>
    Gold:LinearColor=<rgba 0.03 0.15 0.006 0.94>
    R1 DefaultColor=<rgba 0.1593 0.4 0.02 0.95>
    R1 Gold:LinearColor=<rgba 0.0744 0.3 0.018 0.9478>
    R2 DefaultColor:LinearColor=<rgba 0.2453 0.34 0.2108 0.9478>
    R3 DefaultColor:LinearColor=<rgba 0.1584 0.6 0.024 1>
    Use R1 Painted Overrides=true
    Use R2 Painted Overrides=true
    Use R3 Painted Overrides=true
   )
   &VC_InitializeRibbon_v1(
    Color=$Particles.Color
    Custom Offset=true
    Enable Skeletal Position=true
    Lifetime=0.9
    Local Vector Offset=&AddVectorToPosition(Position=<0 0 0>, Vector=$User.DTCarOffset)
    Offset Amount=1
    Ribbon Width=45
   )
   &AddVelocity(Velocity=<-1700 0 0>) {module=AddVelocity001}
  ]
  particleUpdateScript [
   &ParticleState()
   &InheritSourceMovement(ApplyForce=false, ApplyPosition=true, ApplyPositionFalloff=false)
   &SolveForcesAndVelocity(Clamp Velocity=false, Limit Acceleration=false)
   &VC_UpdateRibbon_v1(Enable Skeletal Position=true)
   &ScaleRibbonWidth(Ribbon Width Scale=&FloatFromCurve(FloatCurve=DataInterface() {propertyValues {Curve {
        defaultValue=1
        keys [
         {interpMode=RCIM_Cubic, leaveTangentWeight=0.3985, tangentMode=RCTM_User, time=0, value=1}
         {arriveTangent=-0.0727, arriveTangentWeight=0.3348, interpMode=RCIM_Cubic, leaveTangent=-0.0727, tangentMode=RCTM_User, time=1, value=0.8}
        ]
       }}}))
   /SetVariables(
    Particles.width1:NiagaraFloat=&Multiply_Float(
     A=&FloatFromCurve(FloatCurve=DataInterface() {propertyValues {Curve {
         defaultValue=1
         keys [
          {interpMode=RCIM_Cubic, leaveTangentWeight=0.3985, tangentMode=RCTM_User, time=0, value=1}
          {arriveTangent=-0.0727, arriveTangentWeight=0.3348, interpMode=RCIM_Cubic, leaveTangent=-0.0727, tangentMode=RCTM_User, time=1, value=0.8}
         ]
        }}})
     B=30
    )
    Particles.width3:NiagaraFloat=&Multiply_Float(
     A=&FloatFromCurve(FloatCurve=DataInterface() {propertyValues {Curve {
         defaultValue=1
         keys [
          {interpMode=RCIM_Cubic, leaveTangentWeight=0.3985, tangentMode=RCTM_User, time=0, value=1}
          {arriveTangent=-0.0727, arriveTangentWeight=0.3348, interpMode=RCIM_Cubic, leaveTangent=-0.0727, tangentMode=RCTM_User, time=1, value=0.8}
         ]
        }}})
     B=50
    )
   ) {isSetParameters=true}
  ]
  rendererClasses [&NiagaraRibbonRendererProperties]
  renderers [
   {rendererClass=&NiagaraRibbonRendererProperties}
   {rendererClass=&NiagaraRibbonRendererProperties, rendererIndex=1}
   {rendererClass=&NiagaraRibbonRendererProperties, rendererIndex=2}
   {rendererClass=&NiagaraRibbonRendererProperties, rendererIndex=3}
  ]
 }
 NE_VC_Boost_Sprite_Parent003 {
  emitterUpdateScript [
   &EmitterState(Inactive Response=::NewEnumerator1, Life Cycle Mode=::NewEnumerator1, Loop Duration=0.2, UseLoopDelay=false)
   &VC_FilteredSocketCount()
   &SpawnRate(SpawnRate=50, Use Spawn Probability=false) {enabled=false}
   &SpawnBurst_Instantaneous(Spawn Count=12, Use Spawn Probability=false, UseLoopCountLimit=false)
  ]
  particleSpawnScript [
   &VC_PaintedVariantsOverride(DefaultColor=<rgba 0.3641 0.85 0.017 1>, R1 DefaultColor=<rgba 0.1107 0.4062 0.0122 1>, Use R1 Painted Overrides=true)
   &InitializeParticle(
    Color=$Particles.Color
    Lifetime Max=0.3
    Lifetime Min=0.1
    Lifetime Mode=::NewEnumerator1
    Mesh Scale=&Multiply_VectorByFloat(Vector=&RandomRangeVector(Minimum=<2 0.6 0.9>, Maximum=<3.6 0.8 1.2>), Float=0.8)
    Mesh Scale Mode=::NewEnumerator4
    UsePositionOffset=false
    Write Material Random=true
   )
   &PartitionParticles(Number Of Partitions=$Emitter.PortCount)
   &SkeletalMeshLocation(
    Bone / Socket Index=$Particles.PartitionParticles.Partition
    Bone / Socket Sampling Mode=::NewEnumerator5
    MaskSpawnGroup=false
    Mesh Sampling Type=::NewEnumerator6
    Sampled Position Offset=&AddVectorToPosition(Position=<-10 0 0>, Vector=$User.DTCarOffset)
    Skeletal Mesh=$System.SkeletalMesh
   )
   &AddVelocity(Velocity=<-1500 0 0>) {enabled=false}
   &InitialMeshOrientation(Mesh Orientation Mode=::NewEnumerator2, Orientation Coordinate Space=::Local, Orientation Vector=<-1 0 0>, Use Rotation Vector=false)
   /SetVariables(Particles.conescale:Vector3f=&Multiply_VectorByFloat(Vector=&RandomRangeVector(Minimum=<2 0.6 0.9>, Maximum=<3.6 0.8 1.2>), Float=0.8)) {isSetParameters=true}
  ]
  particleUpdateScript [
   &ParticleState()
   &SolveForcesAndVelocity(Clamp Velocity=false, Limit Acceleration=false) {enabled=false}
   &ScaleColor(Scale Alpha=0.2, Scale RGB=&VectorFromFloat(Value=1.4), ScaleA=true, ScaleRGB=true)
  ]
  rendererClasses [&NiagaraMeshRendererProperties, &NiagaraSpriteRendererProperties]
  renderers [
   {rendererClass=&NiagaraMeshRendererProperties, rendererIndex=0}
   {rendererClass=&NiagaraMeshRendererProperties, rendererIndex=1}
   {rendererClass=&NiagaraSpriteRendererProperties, rendererIndex=2}
  ]
 }
 NE_VC_Boost_Sprite_Parent004 {
  emitterUpdateScript [
   &EmitterState(Inactive Response=::NewEnumerator1, Life Cycle Mode=::NewEnumerator1, Loop Duration=1, UseLoopDelay=false)
   &VC_FilteredSocketCount()
   &SpawnRate(Spawn Probability=0.6, SpawnRate=80, Use Spawn Probability=true)
   &SpawnRate(SpawnGroup=1, SpawnRate=30, Use Spawn Probability=false) {module=SpawnRate001}
  ]
  particleSpawnScript [
   &VC_PaintedVariantsOverride() {enabled=false}
   &InitializeParticle(
    Color Maximum=<rgba 0.0833 0.25 0 1>
    Color Minimum=<rgba 0.0417 0.2 0.01 0.9>
    Color Mode=::NewEnumerator2
    Lifetime Max=0.5
    Lifetime Min=0.4
    Lifetime Mode=::NewEnumerator1
    Sprite Rotation Mode=::NewEnumerator2
    Sprite Size Mode=::NewEnumerator3
    Uniform Sprite Size=&MergeHelper_FloatSwitch(
     0=&RandomRangeFloat(Minimum=80, Maximum=130)
     1=&RandomRangeFloat(Minimum=5, Maximum=15)
     2=?
     3=?
     Index=&SpawnGroup()
     Num=::NewEnumerator0
    )
    UsePositionOffset=false
    Write Material Random=true
   )
   &PartitionParticles(Number Of Partitions=$Emitter.PortCount)
   &SkeletalMeshLocation(
    Bone / Socket Index=$Particles.PartitionParticles.Partition
    Bone / Socket Sampling Mode=::NewEnumerator5
    MaskSpawnGroup=false
    Mesh Sampling Type=::NewEnumerator6
    Skeletal Mesh=$System.SkeletalMesh
   )
   &ShapeLocation(
    Handle Radius=4
    Large Radius=40
    MaskSpawnGroup=false
    Offset=&AddVectorToPosition(Position=<-20 0 0>, Vector=$User.DTCarOffset)
    Offset Coordinate Space=::Local
    Offset Mode=::NewEnumerator0
    Override Seed=false
    Rotation Angle Type:ENiagara_AngleInput=::NewEnumerator0
    Rotation Mode=::NewEnumerator2
    Scale Mode:ENiagara_ScaleMode=::NewEnumerator1
    Shape Primitive=::NewEnumerator3
    Surface Distribution=1
    Yaw / Pitch / Roll=<0 270 0>
   )
   &AddVelocity(Velocity=&RandomRangeVector(Minimum=<-700 0 -20>, Maximum=<-1400 0 50>))
   /SetVariables(Particles.VisibilityTag=$Emitter.SpawnGroup) {isSetParameters=true}
   &SubUVAnimation(
    Evaluation Type=::NewEnumerator0
    "Sprite Renderer "=DataInterface() {propertyValues {SpriteRenderer=&self:NE_VC_Boost_Sprite_Parent004.NiagaraSpriteRendererProperties_0}}
    SubUV Animation Mode:ENiagara_SubUVLookupModeV2=::NewEnumerator2
    UseEndFrame=false
    UseStartFrame=false
   )
  ]
  particleUpdateScript [
   &ParticleState()
   &ScaleSpriteSize(
    Scale Sprite Size Mode=::NewEnumerator0
    Uniform Scale Factor=&MergeHelper_FloatSwitch(
     0=&FloatFromCurve(
      FloatCurve=DataInterface() {propertyValues {Curve {keys [
          {arriveTangent=0.0903, interpMode=RCIM_Cubic, leaveTangent=0.0903, tangentMode=RCTM_User, time=0, value=0.7}
          {arriveTangent=0.6615, interpMode=RCIM_Cubic, leaveTangent=0.6615, tangentMode=RCTM_User, time=1, value=1.2}
         ]}}}
      Scale Curve=&RandomRangeFloat(Minimum=0.9, Maximum=1.2)
     )
     1=&FloatFromCurve(FloatCurve=DataInterface() {propertyValues {Curve {
         defaultValue=1
         keys [
          {arriveTangent=-2.0089, interpMode=RCIM_Cubic, leaveTangent=-2.0089, tangentMode=RCTM_User, time=0, value=1}
          {interpMode=RCIM_Cubic, tangentMode=RCTM_User, time=1, value=0}
         ]
        }}})
     2=?
     3=?
     Index=&SpawnGroup()
     Num=::NewEnumerator0
    )
   )
   &AccelerationForce(Acceleration=&MakeVector(
     X=0
     Y=0
     Z=&Multiply_Float(
      A=&RandomRangeFloat(Minimum=1500, Maximum=2000)
      B=&FloatFromCurve(FloatCurve=DataInterface() {propertyValues {Curve {
          defaultValue=1
          keys [
           {arriveTangent=0.7832, interpMode=RCIM_Cubic, leaveTangent=0.7832, tangentMode=RCTM_User, time=0.1, value=0}
           {arriveTangent=0.5962, interpMode=RCIM_Cubic, leaveTangent=0.5962, tangentMode=RCTM_User, time=1, value=1}
          ]
         }}})
     )
    ))
   &InheritSourceMovement(ApplyForce=false, ApplyPosition=true, ApplyPositionFalloff=false)
   &SolveForcesAndVelocity(Clamp Velocity=false, Limit Acceleration=false)
   &ScaleColor(
    Scale Alpha=&FloatFromCurve(
     FloatCurve=DataInterface() {propertyValues {Curve {keys [
         {arriveTangent=-0.718, interpMode=RCIM_Cubic, leaveTangent=-0.718, leaveTangentWeight=0.3338, tangentMode=RCTM_User, time=0, value=1}
         {arriveTangent=-0.6261, interpMode=RCIM_Cubic, leaveTangent=-0.6261, tangentMode=RCTM_User, time=1, value=0}
        ]}}}
     Scale Curve=&RandomRangeFloat(Minimum=1.1, Maximum=1.3)
    )
    Scale RGB=&VectorFromFloat(Value=&FloatFromCurve(
      FloatCurve=DataInterface() {propertyValues {Curve {keys [
          {arriveTangent=-1.0706, interpMode=RCIM_Cubic, leaveTangent=-1.0706, tangentMode=RCTM_User, time=0, value=3}
          {arriveTangent=-3.9453, interpMode=RCIM_Cubic, leaveTangent=-3.9453, tangentMode=RCTM_User, time=1, value=0}
         ]}}}
      Scale Curve=&RandomRangeFloat(Minimum=1, Maximum=1.5)
     ))
    ScaleA=true
    ScaleRGB=true
   )
  ]
  rendererClasses [&NiagaraSpriteRendererProperties]
  renderers [
   {rendererClass=&NiagaraSpriteRendererProperties}
   {rendererClass=&NiagaraSpriteRendererProperties, rendererIndex=1}
  ]
 }
 NE_VC_Flare {
  emitterUpdateScript [
   &EmitterState(
    Inactive Response=::NewEnumerator1
    Life Cycle Mode=::NewEnumerator1
    Loop Behavior=::NewEnumerator1
    Loop Duration Mode=::NewEnumerator1
    UseLoopDelay=false
   )
   &VC_FilteredSocketCount()
   &SpawnRate(SpawnRate=50, Use Spawn Probability=false) {enabled=false}
   &SpawnBurst_Instantaneous(Spawn Count=4, Use Spawn Probability=false, UseLoopCountLimit=false)
  ]
  particleSpawnScript [
   &VC_PaintedVariantsOverride(DefaultColor=<rgba 0.0904 0.7 0.035 0.5>, R1 DefaultColor=<rgba 0.24 0.6 0.24 0.6>, Use R1 Painted Overrides=true)
   &InitializeParticle(
    Color=$Particles.Color
    Lifetime=0.5
    Position Mode=::NewEnumerator0
    Sprite Size=<1300 70>
    Sprite Size Mode=::NewEnumerator4
    Write Material Random=true
   )
   &PartitionParticles(Number Of Partitions=$Emitter.PortCount)
   &SkeletalMeshLocation(
    Bone / Socket Index=$Particles.PartitionParticles.Partition
    Bone / Socket Sampling Mode=::NewEnumerator5
    MaskSpawnGroup=false
    Mesh Sampling Type=::NewEnumerator6
    Sampled Position Offset=&AddVectorToPosition(Position=<0 0 0>, Vector=$User.DTCarOffset)
    Skeletal Mesh=$System.SkeletalMesh
   )
   &AddVelocity(Velocity=<-1500 0 0>) {enabled=false}
   /SetVariables(Particles.flaresize:Vector2f=<500 100>) {isSetParameters=true}
   &CameraOffset(Camera Offset Amount=20)
  ]
  particleUpdateScript [
   &ParticleState(Kill Particles When Lifetime Has Elapsed=false)
   &SolveForcesAndVelocity(Clamp Velocity=false, Limit Acceleration=false) {enabled=false}
  ]
  rendererClasses [&NiagaraSpriteRendererProperties]
  renderers [
   {rendererClass=&NiagaraSpriteRendererProperties}
   {rendererClass=&NiagaraSpriteRendererProperties, rendererIndex=1}
  ]
 }
 NE_VC_Splash {
  emitterUpdateScript [
   &EmitterState(Inactive Response=::NewEnumerator1, Life Cycle Mode=::NewEnumerator1, Loop Duration=1, UseLoopDelay=false)
   &VC_FilteredSocketCount()
   &SpawnRate(Spawn Probability=0.4, SpawnRate=80, Use Spawn Probability=true)
   &SpawnRate(Spawn Probability=0.7, SpawnGroup=1, SpawnRate=80, Use Spawn Probability=true) {enabled=false, module=SpawnRate001}
  ]
  particleSpawnScript [
   &VC_PaintedVariantsOverride(DefaultColor=<rgba 0.106 0.4 0.008 1>, R1 DefaultColor=<rgba 0.0795 0.3 0.006 1>, Use R1 Painted Overrides=true)
   &InitializeParticle(
    Color=$Particles.Color
    Lifetime Max=0.6
    Lifetime Min=0.4
    Lifetime Mode=::NewEnumerator1
    Sprite Rotation Mode=::NewEnumerator2
    Sprite Size Mode=::NewEnumerator1
    Uniform Sprite Size Max=160
    Uniform Sprite Size Min=140
    UsePositionOffset=false
    Write Material Random=true
   )
   &PartitionParticles(Number Of Partitions=$Emitter.PortCount)
   &SkeletalMeshLocation(
    Bone / Socket Index=$Particles.PartitionParticles.Partition
    Bone / Socket Sampling Mode=::NewEnumerator5
    MaskSpawnGroup=false
    Mesh Sampling Type=::NewEnumerator6
    Sampled Position Offset=&AddVectorToPosition(Position=<-20 0 0>, Vector=$User.DTCarOffset)
    Skeletal Mesh=$System.SkeletalMesh
   )
   &AddVelocity(
    Cone Angle=12
    Cone Axis=<-1 0 0>
    Inner Cone Angle=6
    Module.UseConeFalloff=false
    Module.UseVelDistribution=false
    Override Seed=false
    Velocity Mode=::NewEnumerator2
    Velocity Speed=&MergeHelper_FloatSwitch(
     0=&RandomRangeFloat(Minimum=250, Maximum=500)
     1=&RandomRangeFloat(Minimum=300, Maximum=600)
     2=?
     3=?
     Index=&SpawnGroup()
     Num=::NewEnumerator0
    )
   )
   /SetVariables(Particles.VisibilityTag=$Emitter.SpawnGroup) {isSetParameters=true}
  ]
  particleUpdateScript [
   &ParticleState()
   &ScaleSpriteSize(
    Uniform Curve Scale=1.2
    Uniform Curve Sprite Scale=DataInterface() {propertyValues {Curve {keys [
        {arriveTangent=0.3466, leaveTangentWeight=0.3333, time=0, value=0.65}
        {arriveTangent=0.5611, arriveTangentWeight=0.3333, leaveTangent=0.5611, time=1, value=1.1}
       ]}}}
   )
   &InheritSourceMovement(ApplyForce=false, ApplyPosition=true, ApplyPositionFalloff=false)
   &SolveForcesAndVelocity(Clamp Velocity=false, Limit Acceleration=false)
   &ScaleColor(
    Scale Alpha=&FloatFromCurve(
     FloatCurve=DataInterface() {propertyValues {Curve {
        defaultValue=1
        keys [
         {arriveTangent=2.2095, interpMode=RCIM_Cubic, leaveTangent=2.2095, tangentMode=RCTM_User, time=0, value=0.1}
         {arriveTangent=-1.9222, interpMode=RCIM_Cubic, leaveTangent=-1.9222, tangentMode=RCTM_User, time=0.25, value=0.95}
         {arriveTangent=-0.1993, interpMode=RCIM_Cubic, leaveTangent=-0.1993, tangentMode=RCTM_User, time=1, value=0}
        ]
       }}}
     Scale Curve=0.8
    )
    ScaleA=true
    ScaleRGB=false
   )
   /SetVariables(Particles.R1 Color:LinearColor=&MultiplyLinearColor(
     Linear Color=$Particles.R1 Color
     Scale Factor=&MakeVector4FromVectorAndFloat(
      "Vector (RGB)"=&VectorFromFloat(Value=1)
      "Float (Alpha)"=&FloatFromCurve(
       FloatCurve=DataInterface() {propertyValues {Curve {
          defaultValue=1
          keys [
           {arriveTangent=2.2095, interpMode=RCIM_Cubic, leaveTangent=2.2095, tangentMode=RCTM_User, time=0, value=0.1}
           {arriveTangent=-2.0955, interpMode=RCIM_Cubic, leaveTangent=-2.0955, tangentMode=RCTM_User, time=0.2642, value=0.9423}
           {arriveTangent=-0.2748, interpMode=RCIM_Cubic, leaveTangent=-0.2748, tangentMode=RCTM_User, time=1, value=0}
          ]
         }}}
       Scale Curve=1.3
      )
     )
    )) {enabled=false, isSetParameters=true}
  ]
  rendererClasses [&NiagaraSpriteRendererProperties]
  renderers [
   {rendererClass=&NiagaraSpriteRendererProperties}
   {rendererClass=&NiagaraSpriteRendererProperties, rendererIndex=1}
   {rendererClass=&NiagaraSpriteRendererProperties, rendererIndex=2}
  ]
 }
}
summary {
 systemName=NS_Boost_EP
 emitters [
  {emitterName=NE_VC_Boost, rendererClasses [&NiagaraRibbonRendererProperties]}
  {emitterName=NE_VC_Splash, rendererClasses [&NiagaraSpriteRendererProperties]}
  {
   emitterName=NE_VC_Boost_Sprite_Parent003
   rendererClasses [&NiagaraMeshRendererProperties, &NiagaraSpriteRendererProperties]
  }
  {emitterName=NE_VC_Boost_Sprite_Parent004, rendererClasses [&NiagaraSpriteRendererProperties]}
  {emitterName=NE_VC_Flare, rendererClasses [&NiagaraSpriteRendererProperties]}
 ]
 userVariables [
  User.bIsFrontendPreview {type=NiagaraBool}
  User.bIsPainted {type=NiagaraBool}
  User.ColorVariantID {type=NiagaraInt32}
  User.DTCarOffset {type=Vector3f, description="", defaultValue {struct=Vector3f, value=<0 0 0>}}
  User.GameModeID {type=NiagaraInt32}
  User.PaintedColor {type=LinearColor, description="", defaultValue {struct=LinearColor, value=<rgba 1 1 1 1>}}
  User.PaintedScalar {type=NiagaraFloat, defaultValue=1, description=""}
  User.PortCount {type=NiagaraInt32, defaultValue=2}
 ]
}
